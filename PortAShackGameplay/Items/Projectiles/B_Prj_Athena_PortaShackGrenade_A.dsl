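; Projectile Blueprint for the Port-A-Shack grenade item, adding a build effect on top of logic inherited from its parent class.
; SpawnBuildFX spawns the NS Building Build Niagara system 512 units above a given location, using auto-release pooling and a pre-cull check; nothing in this file calls it.
; The water-entry event has no body, and the construction script only calls the parent.
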(event OnEnterWaterEvent(WaterInteractionComponent)
  (params (WaterBody :type Fort%20Water%20Body%20Actor%20Object%20Reference) (WaterInteractionComponent :type Fort%20Water%20Interaction%20Component%20Object%20Reference) (bIsFirstBody :type Boolean)))

(fn SpawnBuildFX
  (params (Location :type Vector :default "0, 0, 0"))
  (Niagara|SpawnSystematLocation :SystemTemplate (Variables|Default|GetNS%20Building%20Build) :Location (+ Location "0,0,512.000000") :Scale "1.000000,1.000000,1.000000" :bAutoDestroy true :bAutoActivate true :PoolingMethod "AutoRelease" :bPreCullCheck true))

(fn ConstructionScript
  (|CallParentFunction))
